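# Physics asset for the lollipop pickaxe skeletal mesh, a two-handed melee weapon cosmetic.
# It holds a single capsule body on the Melee_TwoHanded_Handle bone, offset along Z and rotated to fit the weapon.
# It has no constraints, so the weapon is represented as one piece with no jointed parts.
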PhysicsAsset /BRCosmetics/Weapons/FORT_Melee/Pickaxe_Lollipop/Meshes/SK_Pickaxe_lollipop_PhysicsAsset
bodies {Melee_TwoHanded_Handle [{shapeType=Capsule, length=89.8936, radius=25.9817, center=<0.8898 -0.66 46.3116>, rotation=<rot -0.3815 -90 -0.3688>}]}
constraints []
